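# Parent material for the RebelClaw Cobalt cosmetic skin, with static switches for glider, pickaxe and backpack variants.
# Character PreFX and PostFX functions bracket optional panning and pattern emissive effects on masked channels.
# Masked regions get time-driven hue or color shifting, with adjusted emissive and triangle-textured roughness.
# Per-item tiled, panning UVs are blended with sphere masks, some using pre-skinned position to separate parts.
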Material /BRCosmetics/Gadgets/Assets/VinderTech_GliderChute/Glider_RebelClaw/Skins/Cobalt/Materials/M_RebelClaw_Cobalt_Parent
graph {
 properties {
  bUseMaterialAttributes=true
  bUsedWithMeshDeformer=true
  bUsedWithNanite=true
  bUsedWithSkeletalMesh=true
  pixelDepthOffsetMode=PDOM_Legacy
  shadingModel=MSM_FromMaterialExpression
  usedShadingModels=Default Lit | Subsurface | Clear Coat | Subsurface Profile | Cloth
 }
 nodes [
  n1 = MaterialFunctionCall() {materialFunction=&MF_CharacterParent_2023_PreFX}
  ChannelMaskParameter(None=@n1.Tech Art Mask) {group=05_Tech Art Masks, parameterName=Panning Emissive Mask}
  n2 = StaticSwitchParameter(
   True=TextureSampleParameter2D() {group=RebelClaw_Cobalt, parameterName=RCC_CustomTexture, texture=&T_BlacksRGB}.RGB
   False=@n1.Tech Art Mask
  ) {group=RebelClaw_Cobalt, parameterName=RCC_UseCustomTexture}
  n3 = StaticComponentMaskParameter(None=@n2) {defaultR=true, group=RebelClaw_Cobalt, parameterName=RCC_Channel, sortPriority=0}
  n4 = ChannelMaskParameter(None=@n3) {group=05_Tech Art Masks, parameterName=Pattern Emissive Mask}
  n5 = Subtract(A=@n4.R, B=NamedRerouteUsage())
  n6 = StaticComponentMaskParameter(None=@n2) {defaultR=true, group=RebelClaw_Cobalt, parameterName=RCC_ColorShifting_Channel, sortPriority=0}
  n7 = NamedRerouteUsage()
  n8 = StaticSwitchParameter(
   True=MaterialFunctionCall!(
    UVOffset
    Alpha=@n6
    isVehicle
    Gmap=ComponentMask(None=@n1.M) {g=true}
    Material=@n7
    Local Normal
    Local Position
    EmissiveColorOverride
    UV=NamedRerouteUsage()
   ) {materialFunction=&/MF_PanningEmissive_CustomUV}.Result
   False=@n7
  ) {group=04_Tech Art, parameterName=UsePanningEmissive, sortPriority=0}
  n9 = MaterialFunctionCall!(
   Alpha=StaticSwitchParameter(True=Saturate(None=@n5), False=@n4.R) {group=RebelClaw_Cobalt, parameterName=RCC_UsePickaxe, sortPriority=1}
   PanningEmissiveBrightnessAdjustment
   CustomUV=NamedRerouteUsage()
   Material=@n8
  ) {materialFunction=&/MF_PatternEmissiveFresnelSequence_CustomUV}
  MaterialFunctionCall(
   Material=StaticSwitchParameter(True=@n9.Result, False=@n8) {group=04_Tech Art, parameterName=UsePatternEmissive, sortPriority=0}
   M=@n1.M
   Pre Skinned Local Position=NamedRerouteDeclaration(None=@n1.Pre skinned Local Position) {name=PreskinLoc, nodeColor=<rgba 1 0 0.5859 1>}
   Distance Blend=@n1.M
   Fresnel=@n1.M
  ) {materialFunction=&MF_CharacterParent_2023_PostFX}
  NamedRerouteDeclaration(None=@n3) {name=Panning_Mask, nodeColor=<rgba 1 0.3047 0 1>}
  n10 = NamedRerouteUsage()
  n11 = ScalarParameter() {defaultValue=0.2, parameterName=RCC_ColorChange_Roughness, sortPriority=2}
  n12 = NamedRerouteUsage()
  n13 = ScalarParameter() {defaultValue=0.15, group=RebelClaw_Cobalt, parameterName=RCC_Roughness1, sortPriority=2}
  n14 = ScalarParameter() {defaultValue=0.3, group=RebelClaw_Cobalt, parameterName=RCC_Roughness2, sortPriority=2}
  n15 = TextureObject() {
   samplerType=SAMPLERTYPE_Masks
   texture=/BRCosmetics/Gadgets/Assets/VinderTech_GliderChute/Glider_RebelClaw/Skins/Cobalt/Textures/T_RebelClaw_Cobalt_Triangles
  }
  n16 = TextureSample(UVs=NamedRerouteUsage(), Tex=@n15) {
   samplerSource=SSM_Wrap_WorldGroupSettings
   samplerType=SAMPLERTYPE_Masks
   texture=&T_RebelClaw_Cobalt_Triangles
  }
  n17 = NamedRerouteDeclaration(None=@n16.R) {name=TrianglesMask, nodeColor=<rgba 0 0.7656 1 1>}
  n18 = NamedRerouteUsage()
  n19 = Saturate(None=Subtract(A=@n18, B=NamedRerouteUsage()))
  NamedRerouteDeclaration(None=LinearInterpolate(
    A=LinearInterpolate(A=@n10, B=@n11, Alpha=@n12)
    B=LinearInterpolate(A=@n13, B=@n14, Alpha=@n17)
    Alpha=StaticSwitchParameter(True=@n19, False=@n18) {group=RebelClaw_Cobalt, parameterName=RCC_UsePickaxe, sortPriority=1}
   )) {name=Roughness_OUT, nodeColor=<rgba 0 0.4375 1 1>}
  n20 = Multiply(A=NamedRerouteUsage(), B=ScalarParameter() {defaultValue=12, group=RebelClaw_Cobalt, parameterName=RCC_UVTiling2})
  n21 = Add(
   A=NamedRerouteUsage()
   B=MaterialFunctionCall(UVs=@n20, "Rotation Angle (0-1)"=ScalarParameter() {group=RebelClaw_Cobalt, parameterName=RCC_UV_Rotation}) {materialFunction=&CustomRotator}.Rotated Values
  )
  n22 = NamedRerouteUsage()
  n23 = ScalarParameter() {defaultValue=7, group=RebelClaw_Cobalt, parameterName=RCC_UV_Tiling_Preskin}
  n24 = MaterialFunctionCall(
   UVs=ComponentMask(None=Divide(A=@n22, B=@n23)) {b=true, g=true}
   "Rotation Angle (0-1)"=ScalarParameter() {defaultValue=0.5, group=RebelClaw_Cobalt, parameterName=RCC_UV_Rotation2}
  ) {materialFunction=&CustomRotator}
  n25 = Multiply(
   A=Time()
   B=ComponentMask(None=VectorParameter() {group=RebelClaw_Cobalt, parameterName=RCC_PanningSpeed2, defaultValue=<rgba 0 0.12 0 1>}.RGB) {g=true, r=true}
  )
  n26 = ComponentMask(None=NamedRerouteUsage()) {b=true}
  n27 = Constant() {r=50}
  NamedRerouteDeclaration(None=LinearInterpolate(A=@n21, B=Add(A=@n24.Rotated Values, B=@n25), Alpha=SphereMask(A=@n26, B=@n27) {attenuationRadius=30})) {name=UVs_Pickaxe, nodeColor=<rgba 0.4062 0 1 1>}
  n28 = Add(
   A=Multiply(A=NamedRerouteUsage(), B=ScalarParameter() {defaultValue=9, group=RebelClaw_Cobalt, parameterName=RCC_UVTiling1})
   B=NamedRerouteUsage()
  )
  NamedRerouteDeclaration(None=@n28) {name=UVs_Character, nodeColor=<rgba 1 0 0.4957 1>}
  n29 = Constant()
  n30 = ScalarParameter() {defaultValue=0.12, group=RebelClaw_Cobalt, parameterName=RCC_PanningSpeed}
  NamedRerouteDeclaration(None=Multiply(A=Time(), B=MaterialFunctionCall(X=@n29, Y=@n30) {materialFunction=&MakeFloat2}.Result)) {name=Panning, nodeColor=<rgba 1 0.5625 0 1>}
  NamedRerouteDeclaration(None=@n1.Result) {name=MA_IN, nodeColor=<rgba 0.7812 1 0 1>}
  n31 = GetMaterialAttributes(None=@n1.Result) {attributeGetTypes [#1, #2, #3]}
  NamedRerouteDeclaration(None=@n31.Emissive Color) {name=Emissive_IN, nodeColor=<rgba 1 0.75 0 1>}
  n32 = NamedRerouteUsage()
  n33 = ScalarParameter() {defaultValue=0.04, group=RebelClaw_Cobalt, parameterName=RCC_ColorChange_HueShift}
  n34 = NamedRerouteUsage()
  n35 = Multiply(A=Time(), B=ScalarParameter() {defaultValue=0.2, group=RebelClaw_Cobalt, parameterName=RCC_ColorChange_Speed})
  n36 = NamedRerouteUsage()
  n37 = Subtract(
   A=LinearInterpolate(
    A=@n32
    B=MaterialFunctionCall(Hue Shift Percentage=@n33, Texture=@n34) {materialFunction=&HueShift}.Result
    Alpha=Sine(None=@n35)
   )
   B=Multiply(A=OneMinus(None=@n36), B=ScalarParameter() {group=RebelClaw_Cobalt, parameterName=RCC_EmissiveDarkening})
  )
  n38 = Multiply(A=NamedRerouteUsage(), B=ScalarParameter() {defaultValue=0.35, group=RebelClaw_Cobalt, parameterName=RCC_EmissiveIntensity})
  n39 = NamedRerouteUsage()
  n40 = Saturate(None=Subtract(A=@n39, B=NamedRerouteUsage()))
  NamedRerouteDeclaration(None=LinearInterpolate(
    A=@n32
    B=Add(A=@n37, B=@n38)
    Alpha=StaticSwitchParameter(True=@n40, False=@n39) {group=RebelClaw_Cobalt, parameterName=RCC_UsePickaxe, sortPriority=1}
   )) {name=Emissive_OUT, nodeColor=<rgba 1 0.0234 0 1>}
  n41 = Normalize(VectorInput=VertexInterpolator(VS=PreSkinnedNormal()).PS)
  n42 = Saturate(None=Abs(None=ComponentMask(None=@n41) {b=true}))
  n43 = Constant() {r=1}
  NamedRerouteDeclaration(None=ShaderStageSwitch(
    PixelShader=MaterialFunctionCall(In=@n42, Contrast=@n43) {materialFunction=&CheapContrast}.Result
    VertexShader=Constant()
   )) {name=PlanarBlend, nodeColor=<rgba 0.7344 1 0 1>}
  NamedRerouteDeclaration(None=@n6) {name=Shifting_Mask, nodeColor=<rgba 1 0.6797 0 1>}
  n44 = NamedRerouteUsage()
  n45 = Sine(None=Multiply(A=Time(), B=ScalarParameter() {defaultValue=0.2, group=RebelClaw_Cobalt, parameterName=RCC_ColorChange_Speed}))
  n46 = LinearInterpolate(
   A=Constant3Vector() {constant=<rgba 0 0.1311 0.8698 1>}
   B=Constant3Vector() {constant=<rgba 0.3659 0.0158 0.4896 1>}
   Alpha=ConstantBiasScale(None=@n45)
  )
  n47 = LinearInterpolate(
   A=@n44
   B=MaterialFunctionCall(
    Hue Shift Percentage=ScalarParameter() {defaultValue=0.04, group=RebelClaw_Cobalt, parameterName=RCC_ColorChange_HueShift}
    Texture=NamedRerouteUsage()
   ) {materialFunction=&HueShift}.Result
   Alpha=@n45
  )
  n48 = Multiply(A=OneMinus(None=NamedRerouteUsage()), B=NamedRerouteUsage())
  n49 = ScalarParameter() {defaultValue=0.035, group=RebelClaw_Cobalt, parameterName=RCC_ColorChange_Darkening}
  n50 = Multiply(
   A=Subtract(A=StaticSwitchParameter(True=@n46, False=@n47) {group=RebelClaw_Cobalt, parameterName=UseCustomTransitionColor}, B=Multiply(A=@n48, B=@n49))
   B=NamedRerouteUsage()
  )
  n51 = NamedRerouteUsage()
  n52 = Saturate(None=Subtract(A=@n51, B=NamedRerouteUsage()))
  NamedRerouteDeclaration(None=LinearInterpolate(
    A=@n44
    B=NamedRerouteDeclaration(None=@n50) {name=BaseColor, nodeColor=<rgba 0.3125 1 0 1>}
    Alpha=StaticSwitchParameter(True=@n52, False=@n51) {group=RebelClaw_Cobalt, parameterName=RCC_UsePickaxe, sortPriority=1}
   ) {constB=0}) {name=BaseColor_OUT, nodeColor=<rgba 0.7813 0 1 1>}
  NamedRerouteDeclaration(None=@n31.Base Color) {name=Basecolor_IN, nodeColor=<rgba 1 0 0.7969 1>}
  n53 = Multiply(
   A=Fresnel(ExponentIn=ScalarParameter() {defaultValue=1, group=RebelClaw_Cobalt, parameterName=RCC_FresnelExp}) {baseReflectFraction=0, exponent=2}
   B=ScalarParameter() {defaultValue=0.5, group=RebelClaw_Cobalt, parameterName=RCC_FresnelIntensity}
  )
  NamedRerouteDeclaration(None=OneMinus(None=Saturate(None=@n53))) {name=Fresnel, nodeColor=<rgba 0.5469 1 0 1>}
  NamedRerouteDeclaration(None=@n31.Roughness) {name=Roughness_IN, nodeColor=<rgba 0 0.1947 1 1>}
  n54 = NamedRerouteUsage()
  n55 = StaticSwitchParameter(True=NamedRerouteUsage(), False=NamedRerouteUsage()) {group=RebelClaw_Cobalt, parameterName=RCC_UseGlider, sortPriority=1}
  NamedRerouteDeclaration(None=StaticSwitchParameter(
    True=NamedRerouteUsage()
    False=StaticSwitchParameter(True=@n54, False=@n55) {group=RebelClaw_Cobalt, parameterName=RCC_UsePickaxe, sortPriority=1}
   ) {group=RebelClaw_Cobalt, parameterName=RCC_UseBackpack, sortPriority=1}) {name=UVs_Final, nodeColor=<rgba 1 0 0.5156 1>}
  NamedRerouteDeclaration(None=TextureCoordinate()) {name=UVs, nodeColor=<rgba 1 0 0.0469 1>}
  n56 = NamedRerouteUsage()
  n57 = MaterialFunctionCall(UVs=@n20, "Rotation Angle (0-1)"=ScalarParameter() {group=RebelClaw_Cobalt, parameterName=RCC_UV_Rotation}) {materialFunction=&CustomRotator}
  n58 = ComponentMask(None=TextureCoordinate()) {g=true}
  n59 = Constant() {r=0.1}
  n60 = Constant() {r=0.2}
  NamedRerouteDeclaration(None=LinearInterpolate(A=@n28, B=Add(A=@n56, B=@n57.Rotated Values), Alpha=SphereMask(A=@n58, B=@n59, Radius=@n60))) {name=UVs_Backpack, nodeColor=<rgba 0 0.3672 1 1>}
  n61 = Add(A=NamedRerouteUsage(), B=@n20)
  n62 = SphereMask(A=ComponentMask(None=NamedRerouteUsage()) {g=true}, B=Constant() {r=120}) {attenuationRadius=100}
  n63 = ComponentMask(None=NamedRerouteUsage()) {g=true}
  n64 = Constant() {r=185}
  NamedRerouteDeclaration(None=LinearInterpolate(A=LinearInterpolate(A=@n28, B=@n61, Alpha=@n62), B=@n21, Alpha=SphereMask(A=@n63, B=@n64) {attenuationRadius=50})) {name=UVs_Glider, nodeColor=<rgba 1 0 0.7266 1>}
  NamedRerouteUsage()
  StaticSwitchParameter(
   True=SetMaterialAttributes!(
    MaterialAttributes=@n31.MaterialAttributes
    Roughness=NamedRerouteUsage()
    Base Color=NamedRerouteUsage()
    Emissive Color=NamedRerouteUsage()
   ) {attributeSetTypes [#1, #2, #3]}
   False=NamedRerouteUsage()
  ) {group=RebelClaw_Cobalt, parameterName=UseRebelClawCobalt}
 ]
}
